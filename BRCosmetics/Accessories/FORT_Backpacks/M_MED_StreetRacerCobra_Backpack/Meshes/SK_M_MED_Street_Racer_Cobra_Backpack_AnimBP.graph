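# Animation Blueprint for the Street Racer Cobra backpack cosmetic that adds secondary physics motion to its bat, handle and straps.
# The input pose passes through four chained single-bone Anim Dynamics nodes with box bodies, angular limits and high damping.
# The bat uses a stiff angular spring blended at partial alpha, while handle and straps run at full alpha.
# The EventGraph has no logic, and a physics-test Sequence Player sits unconnected.
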AnimBlueprint /BRCosmetics/Accessories/FORT_Backpacks/M_MED_StreetRacerCobra_Backpack/Meshes/SK_M_MED_Street_Racer_Cobra_Backpack_AnimBP
target_skeleton=/Game/Accessories/FORT_Backpacks/M_MED_StreetRacerCobra_Backpack/Meshes/SK_M_MED_Street_Racer_Cobra_Backpack_Skeleton
graphs [
 AnimGraph {nodes [
   AnimGraphNode_SubInput_0 = Input Pose()
   n1 = Anim Dynamics - Bone: dyn_bat(ComponentPose=Local To Component(LocalPose=@AnimGraphNode_SubInput_0), Alpha=0.35, ExternalForce="0,0,0") {
    angularDampingOverride=0.99
    angularSpringConstant=3500
    bAngularSpring=true
    bGravityOverrideInSimSpace=true
    bOverrideAngularDamping=true
    numSolverIterationsPostUpdate=4
    numSolverIterationsPreUpdate=16
    alphaBoolBlend {blendOption=Linear}
    boundBone {boneName=dyn_bat}
    physicsBodyDefinitions [{
      sphereCollisionRadius=0
      boundBone {boneName=dyn_bat}
      boxExtents=<20 20 10>
      constraintSetup {angularTargetAxis=AxisY, twistAxis=AxisZ, angularLimitsMax=<40 20 5>, angularLimitsMin=<-15 -20 -5>, angularTarget=<0 1 0>}
      localJointOffset=<0 0 -3>
     }]
   }
   n2 = Anim Dynamics - Bone: dyn_handle(ComponentPose=@n1, Alpha=1, ExternalForce="0,0,0") {
    angularDampingOverride=0.95
    bGravityOverrideInSimSpace=true
    bOverrideAngularDamping=true
    numSolverIterationsPostUpdate=4
    numSolverIterationsPreUpdate=16
    alphaBoolBlend {blendOption=Linear}
    boundBone {boneName=dyn_handle}
    componentLinearAccScale=<1 1 1>
    componentLinearVelScale=<1 1 1>
    physicsBodyDefinitions [{
      sphereCollisionRadius=0
      boundBone {boneName=dyn_handle}
      boxExtents=<10 10 5>
      constraintSetup {twistAxis=AxisZ, angularLimitsMax=<30 0 0>, angularLimitsMin=<-5 0 0>, angularTarget=<0 0 0>}
      localJointOffset=<0 0 -1>
     }]
   }
   n3 = Anim Dynamics - Bone: dyn_strap_top(ComponentPose=@n2, Alpha=1, ExternalForce="0,0,0") {
    angularDampingOverride=0.95
    bGravityOverrideInSimSpace=true
    bOverrideAngularDamping=true
    numSolverIterationsPostUpdate=4
    numSolverIterationsPreUpdate=16
    alphaBoolBlend {blendOption=Linear}
    boundBone {boneName=dyn_strap_top}
    componentLinearAccScale=<1 1 1>
    componentLinearVelScale=<1 1 1>
    physicsBodyDefinitions [{
      sphereCollisionRadius=0
      boundBone {boneName=dyn_strap_top}
      boxExtents=<20 8 8>
      constraintSetup {angularLimitsMax=<5 0 30>, angularLimitsMin=<-5 0 -30>, angularTarget=<0 0 0>}
      localJointOffset=<2 0 0>
     }]
   }
   n4 = Anim Dynamics - Bone: dyn_strap_bott(ComponentPose=@n3, Alpha=1, ExternalForce="0,0,0") {
    angularDampingOverride=0.95
    bGravityOverrideInSimSpace=true
    bOverrideAngularDamping=true
    numSolverIterationsPostUpdate=4
    numSolverIterationsPreUpdate=16
    alphaBoolBlend {blendOption=Linear}
    boundBone {boneName=dyn_strap_bott}
    componentLinearAccScale=<1 1 1>
    componentLinearVelScale=<1 1 1>
    physicsBodyDefinitions [{
      sphereCollisionRadius=0
      boundBone {boneName=dyn_strap_bott}
      boxExtents=<20 8 8>
      constraintSetup {angularLimitsMax=<5 0 30>, angularLimitsMin=<-5 0 -30>, angularTarget=<0 0 0>}
      localJointOffset=<2 0 0>
     }]
   }
   Output Pose(Result=Component To Local(ComponentPose=@n4)) {layerGroup=None}
   Sequence Player 'Backpack_PhysicsTest_Animation'() {sequence=/Game/Accessories/FORT_Backpacks/M_MED_StreetRacerCobra_Backpack/Meshes/Backpack_PhysicsTest_Animation}
  ]}
 EventGraph {nodes [Event BlueprintUpdateAnimation()]}
]
variables []
